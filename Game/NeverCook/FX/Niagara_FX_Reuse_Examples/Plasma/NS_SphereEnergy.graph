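# Niagara plasma effect NS_SphereEnergy, filed among FX reuse examples, that renders a dark purple energy ball as a mesh.
# Its single Ball emitter spawns by instantaneous burst, and particles are not killed when their lifetime elapses.
# ScaleColor drives RGB and alpha, while a sprite-size scale module is also present despite the mesh-only renderer.
# A User.VariantSwitch float is exposed, with no visible link to the listed modules.
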NiagaraSystem /Game/NeverCook/FX/Niagara_FX_Reuse_Examples/Plasma/NS_SphereEnergy
dependencies {usedDataInterfaces [&NiagaraDataInterfaceCurve], usedDynamicInputs [], usedModules [&SystemState]}
emitters {Ball {
  emitterUpdateScript [
   &EmitterState(
    Inactive Response=::NewEnumerator1
    Life Cycle Mode=::NewEnumerator1
    Loop Behavior=::NewEnumerator1
    Loop Duration Mode=::NewEnumerator1
    UseLoopDelay=true
   )
   &SpawnBurst_Instantaneous(Use Spawn Probability=false, UseLoopCountLimit=false)
  ]
  particleSpawnScript [&InitializeParticle(
    Color=<rgba 0.0605 0.02 0.15 1>
    Lifetime=5.6
    Mesh Scale Mode=::NewEnumerator3
    Sprite Size Mode=::NewEnumerator3
    Uniform Sprite Size=120
    UsePositionOffset=false
    Write Material Random=true
   )]
  particleUpdateScript [
   &ParticleState(Kill Particles When Lifetime Has Elapsed=false, Let Infinitely Lived Particles Die When Emitter Deactivates=true)
   &ScaleSpriteSize(Scale Sprite Size Mode=::NewEnumerator0) {module=ScaleSpriteSize001}
   &ScaleColor(ScaleA=true, ScaleRGB=true)
  ]
  rendererClasses [&NiagaraMeshRendererProperties]
  renderers [{rendererClass=&NiagaraMeshRendererProperties, rendererIndex=0}]
 }}
summary {
 systemName=NS_SphereEnergy
 emitters [{emitterName=Ball, rendererClasses [&NiagaraMeshRendererProperties]}]
 userVariables [User.VariantSwitch {type=NiagaraFloat, defaultValue=0, description=""}]
}
